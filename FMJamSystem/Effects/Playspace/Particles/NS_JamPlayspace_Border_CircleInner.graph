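# Niagara effect for the inner circle border of a Jam playspace: a tempo-reactive mesh ring with small dust sprites.
# CircleInner bursts a persistent, slowly rotating mesh that scales and fades in.
# Its brightness spikes then decays over one beat, computed from engine time since User.GameTime and User.Tempo.
# Small_Dust spawns track-colored, rising, shrinking sprites on a ring, with rate and radius driven by RingSize and alpha dimmed by time of day.
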NiagaraSystem /FMJamSystem/Effects/Playspace/Particles/NS_JamPlayspace_Border_CircleInner
dependencies {
 usedDataInterfaces [
  &NiagaraDataInterfaceVectorCurve
  &NiagaraDataInterfaceCurve
  &NiagaraDataInterfaceVector2DCurve
  &NiagaraDataInterfaceSkeletalMesh
 ]
 usedDynamicInputs [&RandomRangeFloat]
 usedModules [
  &self:SystemSpawnScript.SystemScriptSource.SystemScriptGraph.NiagaraNodeAssignment_0.SetVariables_#1
  &SystemState
 ]
}
emitters {
 CircleInner {
  emitterUpdateScript [
   &EmitterState(Life Cycle Mode=::NewEnumerator1, Loop Behavior=::NewEnumerator1, Loop Duration Mode=::NewEnumerator1, UseLoopDelay=false)
   &SpawnBurst_Instantaneous(Use Spawn Probability=false, UseLoopCountLimit=false)
  ]
  particleSpawnScript [
   &InitializeParticle(
    Lifetime=0.5
    Mesh Scale Mode=::NewEnumerator4
    Position Offset=<0 0 20>
    Position Offset Coordinate Space=::Local
    UsePositionOffset=true
    Write Material Random=true
   )
   &InitialMeshOrientation(
    Mesh Axis To Orient=<0 0 0>
    Mesh Orientation Mode=::NewEnumerator2
    Orientation Vector=&RandomRangeVector(Minimum=<0 -1 0>, Maximum=<0 1 0>)
    Use Rotation Vector=false
   )
  ]
  particleUpdateScript [
   &ParticleState(Kill Particles When Lifetime Has Elapsed=false, Let Infinitely Lived Particles Die When Emitter Deactivates=true)
   &ScaleMeshSize(Scale Factor=&VectorFromCurve(VectorCurve=DataInterface() {propertyValues {
       XCurve {keys [
         {arriveTangent=6.2185, interpMode=RCIM_Cubic, leaveTangent=6.2185, tangentMode=RCTM_User, time=0, value=0}
         {arriveTangent=0.2679, interpMode=RCIM_Cubic, leaveTangent=0.2679, tangentMode=RCTM_User, time=0.4, value=1}
         {arriveTangent=0.2392, interpMode=RCIM_Cubic, leaveTangent=0.2392, tangentMode=RCTM_User, time=0.6, value=1}
        ]}
       YCurve {keys [
         {arriveTangent=6.2185, interpMode=RCIM_Cubic, leaveTangent=6.2185, tangentMode=RCTM_User, time=0, value=0}
         {arriveTangent=0.2679, interpMode=RCIM_Cubic, leaveTangent=0.2679, tangentMode=RCTM_User, time=0.4, value=1}
         {arriveTangent=0.2392, interpMode=RCIM_Cubic, leaveTangent=0.2392, tangentMode=RCTM_User, time=0.6, value=1}
        ]}
       ZCurve {keys [
         {arriveTangent=1.68, interpMode=RCIM_Cubic, leaveTangent=1.68, tangentMode=RCTM_User, time=0, value=0}
         {arriveTangent=-0.0952, interpMode=RCIM_Cubic, leaveTangent=-0.0952, tangentMode=RCTM_User, time=0.35, value=1.5}
         {arriveTangent=-0.0405, interpMode=RCIM_Cubic, leaveTangent=-0.0405, tangentMode=RCTM_User, time=0.6, value=1}
        ]}
      }}))
   &UpdateMeshOrientation(Rotation Rate=0.002, Rotation Vector=<0 0 1>) {module=UpdateMeshOrientation001}
   &ScaleColor(
    Scale Alpha=&FloatFromCurve(FloatCurve=DataInterface() {propertyValues {Curve {keys [
         {arriveTangent=4.1785, interpMode=RCIM_Cubic, leaveTangent=4.1785, tangentMode=RCTM_User, time=0, value=0}
         {arriveTangent=1.1179, interpMode=RCIM_Cubic, leaveTangent=1.1179, tangentMode=RCTM_User, time=0.2042, value=0.6228}
         {arriveTangent=0.0224, interpMode=RCIM_Cubic, leaveTangent=0.0224, tangentMode=RCTM_User, time=1, value=1}
        ]}}})
    Scale RGB=&VectorFromFloat(Value=&FloatFromCurve(
      FloatCurve=DataInterface() {propertyValues {Curve {keys [
          {arriveTangent=-0.3618, interpMode=RCIM_Cubic, leaveTangent=-0.3618, tangentMode=RCTM_User, time=0, value=0.25}
          {interpMode=RCIM_Cubic, tangentMode=RCTM_User, time=0.01, value=0.6}
          {arriveTangent=-0.0405, interpMode=RCIM_Cubic, leaveTangent=-0.0405, tangentMode=RCTM_User, time=1, value=0.25}
         ]}}}
      CurveIndex=&ClampFloat(Float=&Divide_Float(
        A=&Subtract_Float(A=$Engine.Time, B=$User.GameTime)
        B=&Divide_Float(A=60, B=&MakeFloatFromInt(Int=$User.Tempo))
       ))
     ))
    ScaleA=true
    ScaleRGB=true
   )
  ]
  rendererClasses [&NiagaraMeshRendererProperties]
  renderers [{rendererClass=&NiagaraMeshRendererProperties, rendererIndex=0}]
 }
 Small_Dust {
  emitterUpdateScript [
   &EmitterState()
   &SpawnRate(SpawnRate=&Multiply_Float(A=20, B=$User.RingSize), Use Spawn Probability=false)
  ]
  particleSpawnScript [
   &InitializeParticle(
    Color=$User.TrackColor
    Lifetime Min=0.5
    Lifetime Mode=::NewEnumerator1
    Mass Mode=::NewEnumerator3
    Position Offset=<0 0 20>
    Sprite Rotation Mode=::NewEnumerator2
    Sprite Size Mode=::NewEnumerator1
    Uniform Sprite Size Max=2
    Uniform Sprite Size Min=1
    UsePositionOffset=true
    Write Material Random=true
   )
   &AddVelocity(Velocity=<0 0 20>)
   &ShapeLocation(
    MaskSpawnGroup=false
    Override Seed=false
    Ring Radius=&Multiply_Float(A=$User.RingSize, B=875)
    Shape Primitive=::NewEnumerator4
    UseOwnerScale=true
   )
  ]
  particleUpdateScript [
   &ParticleState()
   &ScaleSpriteSize(
    Non-Uniform Curve Sprite Scale=DataInterface() {propertyValues {
      XCurve {keys [{interpMode=RCIM_Cubic, tangentMode=RCTM_User, time=0, value=1}, {interpMode=RCIM_Cubic, tangentMode=RCTM_User, time=1, value=0}]}
      YCurve {keys [{interpMode=RCIM_Cubic, tangentMode=RCTM_User, time=0, value=1}, {interpMode=RCIM_Cubic, tangentMode=RCTM_User, time=1, value=0}]}
     }}
    Scale Sprite Size Mode=::NewEnumerator3
   )
   &Drag(Use Linear Drag=true, Use Rotational Drag=false)
   &ScaleColor(
    Scale Alpha=&Multiply_Float(
     A=&FNTimeOfDayScalar(Night Scale=0.5, Morning Scale=0.8, Evening Scale=0.8)
     B=&FloatFromCurve(FloatCurve=DataInterface() {propertyValues {Curve {keys [{interpMode=RCIM_Cubic, tangentMode=RCTM_User, time=0, value=1}, {interpMode=RCIM_Cubic, tangentMode=RCTM_User, time=1, value=0}]}}})
    )
    ScaleA=true
    ScaleRGB=true
   )
   &SolveForcesAndVelocity(Clamp Velocity=false, Limit Acceleration=false)
  ]
  rendererClasses [&NiagaraSpriteRendererProperties]
  renderers [{rendererClass=&NiagaraSpriteRendererProperties}]
 }
}
summary {
 systemName=NS_JamPlayspace_Border_CircleInner
 emitters [
  {emitterName=CircleInner, rendererClasses [&NiagaraMeshRendererProperties]}
  {emitterName=Small_Dust, rendererClasses [&NiagaraSpriteRendererProperties]}
 ]
 userVariables [
  User.GameTime {type=NiagaraFloat, defaultValue=0, description=""}
  User.RingSize {type=NiagaraFloat, defaultValue=1, description=""}
  User.Tempo {type=NiagaraInt32}
  User.TrackColor {type=LinearColor, description="", defaultValue {struct=LinearColor, value=<rgba 0.125 0.91 0.18 1>}}
 ]
}
